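# Translucent UI material that draws a hollow circular progress bar with a faint outline and optional backing disc.
# A PolarCoords texture supplies radius for DDX-antialiased SmoothStep rings and angle for the sweep, which can fill or empty.
# Static switches choose fill direction and whether progress comes from Percentage or is calculated from Time, start time and duration.
# HasMultipleInteracts blends in a sine-pulsed radial glow and swaps the base color for glow colors.
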Material /Game/UI/Materials/Athena_HollowCircularProgressBar
graph {
 properties {
  bUsedWithStaticMesh=false
  blendMode=BLEND_Translucent
  materialDomain=MD_UI
  pixelDepthOffsetMode=PDOM_Legacy
  usedShadingModels=Default Lit
 }
 nodes [
  n1 = TextureSample() {samplerType=SAMPLERTYPE_LinearColor, texture=&/PolarCoords}
  n2 = Constant() {r=0.91}
  n3 = ComponentMask(None=TextureCoordinate()) {r=true}
  n4 = Multiply(A=Abs(None=DDX(Value=@n3))) {constB=6}
  n5 = MaterialFunctionCall(Alpha=@n1.B, Min=@n2, Max=Add(A=@n2, B=@n4)) {materialFunction=&SmoothStep}
  n6 = Constant() {r=0.97}
  n7 = MaterialFunctionCall(Alpha=@n1.B, Min=@n6, Max=Add(A=@n6, B=@n4)) {materialFunction=&SmoothStep}
  n8 = OneMinus(None=@n5.Result)
  n9 = ScalarParameter() {parameterName=Back Alpha, sliderMax=1}
  n10 = TextureSample() {samplerType=SAMPLERTYPE_LinearColor, texture=&PolarCoords}
  n11 = ScalarParameter() {defaultValue=0.68, parameterName=ProgressBarThickness}
  n12 = ComponentMask(None=TextureCoordinate()) {r=true}
  n13 = Multiply(A=Abs(None=DDX(Value=@n12))) {constB=4}
  n14 = ScalarParameter() {defaultValue=0.83, parameterName=ProgressBarSize}
  n15 = Subtract(
   A=MaterialFunctionCall(Alpha=@n10.B, Min=@n11, Max=Add(A=@n11, B=@n13)) {materialFunction=&SmoothStep}.Result
   B=MaterialFunctionCall(Alpha=@n10.B, Min=@n14, Max=Add(A=@n14, B=@n13)) {materialFunction=&SmoothStep}.Result
  )
  n16 = ScalarParameter() {defaultValue=0.45, parameterName=GlowInner_Radius}
  n17 = ScalarParameter() {defaultValue=3, parameterName=GlowInner_Density}
  n18 = DotProduct(
   A=MaterialFunctionCall(
    Radius=ScalarParameter() {defaultValue=0.5, parameterName=GlowOuter_Radius}
    Density=ScalarParameter() {defaultValue=5, parameterName=GlowOuter_Density}
   ) {materialFunction=&RadialGradientExponential}.RadialGradientExponential
   B=OneMinus(None=MaterialFunctionCall(Radius=@n16, Density=@n17) {materialFunction=&RadialGradientExponential}.RadialGradientExponential)
  )
  n19 = Sine(None=Multiply(A=Time(), B=ScalarParameter() {defaultValue=1.25, parameterName=GlowSpeed}))
  n20 = LinearInterpolate(
   A=ScalarParameter() {defaultValue=0.33, parameterName=GlowAlpha_Min}
   B=ScalarParameter() {defaultValue=1, parameterName=GlowAlpha_Max}
   Alpha=Multiply(A=Add(A=@n19)) {constB=0.5}
  ) {constA=2.5, constB=3}
  n21 = LinearInterpolate(A=@n15, B=LinearInterpolate(A=Multiply(A=@n18, B=@n20), Alpha=@n15), Alpha=ScalarParameter() {parameterName=HasMultipleInteracts})
  n22 = Subtract(A=Time(), B=ScalarParameter() {parameterName=CountdownStartTime})
  n23 = ScalarParameter() {defaultValue=1, parameterName=CountDownTime}
  n24 = StaticSwitchParameter(True=Clamp(None=Divide(A=@n22, B=@n23)), False=LinearInterpolate(Alpha=ScalarParameter() {parameterName=Percentage}) {constA=-0.001}) {
   desc="In Unattended mode you give the progress bar a max time, and a start time, and it automatically cycles up."
   parameterName=Calculate Progress
  }
  n25 = Add(A=TextureSample() {samplerType=SAMPLERTYPE_LinearColor, texture=&PolarCoords}.R, B=@n24)
  n26 = StaticSwitchParameter(True=Floor(None=@n25), False=Ceil(None=OneMinus(None=@n25))) {defaultValue=true, desc=Whether the progress bar fills or empties based on the percentage., parameterName=Fills}
  n27 = Add(
   A=Multiply(A=ScalarParameter() {defaultValue=0.2, parameterName=OutlineAlpha, sliderMax=1}, B=Subtract(A=@n5.Result, B=@n7.Result))
   B=Add(A=Multiply(A=@n8, B=@n9), B=Multiply(A=@n21, B=@n26))
  )
  LinearInterpolate(
   A=VectorParameter() {parameterName=Color, defaultValue=<rgba 1 1 1 1>}.RGB
   B=LinearInterpolate(
    A=VectorParameter() {parameterName=GlowColor_Falloff, defaultValue=<rgba 0 1 0.25 1>}.RGB
    B=VectorParameter() {parameterName=GlowColor, defaultValue=<rgba 0.4531 1 0.688 1>}.RGB
    Alpha=@n27
   )
   Alpha=ScalarParameter() {parameterName=HasMultipleInteracts}
  )
  Multiply(A=@n27, B=TextureSample() {samplerType=SAMPLERTYPE_LinearColor, texture=&PolarCoords}.A)
 ]
}
