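# MF_RingMask is a library-exposed utility material function that outputs a procedural ring mask from Coordinates, Center, OuterRadius, InnerRadiusWeight and Hardness.
# A zero-hardness SphereMask gives a linear radial falloff, scaled by InnerRadiusWeight to set band thickness.
# Subtracting that from its doubled, saturated copy forms a triangular band, which Hardness sharpens before a final saturate.
# The inner radius ignores Hardness to save preshader operations.
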MaterialFunction /Game/UI/Materials/MF_RingMask
graph {
 properties {bExposeToLibrary=true, libraryCategoriesText [Utility]}
 nodes [
  n1 = FunctionInput(Preview=TextureCoordinate()) {bUsePreviewValueAsDefault=true, inputName=Coordinates, inputType=FunctionInput_Vector2, previewValue=<0.5 0 0 1>}
  n2 = FunctionInput() {bUsePreviewValueAsDefault=true, inputName=Center, inputType=FunctionInput_Vector2, sortPriority=1, previewValue=<0.5 0.5 0 1>}
  n3 = FunctionInput() {bUsePreviewValueAsDefault=true, inputName=OuterRadius, inputType=FunctionInput_Scalar, sortPriority=2, previewValue=<0.5 0 0 1>}
  n4 = FunctionInput() {bUsePreviewValueAsDefault=true, inputName=InnerRadiusWeight, inputType=FunctionInput_Scalar, sortPriority=3, previewValue=<8 0 0 1>}
  n5 = Multiply(A=SphereMask(A=@n1, B=@n2, Radius=@n3) {desc=Not adjusting inner radius on Hardness to save on Preshader ops, hardnessPercent=0}, B=@n4) {constB=2}
  n6 = Subtract(A=Saturate(None=Multiply(A=@n5) {constB=2}), B=@n5)
  n7 = FunctionInput() {bUsePreviewValueAsDefault=true, inputName=Hardness, inputType=FunctionInput_Scalar, sortPriority=4, previewValue=<5 0 0 1>}
  FunctionOutput(None=Saturate(None=Multiply(A=@n6, B=@n7) {constB=3}))
 ]
}
