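# Translucent UI material for a circular progress bar driven by a Percentage parameter.
# Centered, negated, axis-swapped UVs become a radial angle, which is offset by Percentage and rounded into a hard sweep mask.
# The static switch Fills picks whether the sweep fills or empties.
# The sweep multiplies FillColor for color and MF_RingMask, shaped by InnerRadiusWeight and Hardness, for the ring shape.
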Material /Game/UI/Materials/CircularProgressBar
graph {
 properties {
  bUsedWithStaticMesh=false
  blendMode=BLEND_Translucent
  materialDomain=MD_UI
  pixelDepthOffsetMode=PDOM_Legacy
  usedShadingModels=Default Lit
 }
 nodes [
  n1 = Multiply(A=Subtract(A=TextureCoordinate()) {constB=0.5}) {constB=-1}
  n2 = Add(
   A=MaterialFunctionCall(Vector or UVs=MaterialFunctionCall(XY=@n1) {materialFunction=&Swizzle}.YX) {materialFunction=&VectorToRadialValue}.Vector Converted to Angle
   B=ScalarParameter() {defaultValue=0.25, group=Progress, parameterName=Percentage, sliderMax=1}
  )
  n3 = StaticSwitchParameter(True=Floor(None=@n2), False=Ceil(None=OneMinus(None=@n2))) {defaultValue=true, desc=Whether the progress bar fills or empties based on the percentage., parameterName=Fills}
  Multiply(A=VectorParameter() {group=Color, parameterName=FillColor, defaultValue=<rgba 1 1 1 1>}.RGB, B=@n3)
  Multiply(
   A=@n3
   B=MaterialFunctionCall!(
    Coordinates
    Center
    OuterRadius
    InnerRadiusWeight=ScalarParameter() {defaultValue=8, group=Shape, parameterName=InnerRadiusWeight}
    Hardness=ScalarParameter() {defaultValue=8, group=Shape, parameterName=Hardness, sliderMax=16, sliderMin=1}
   ) {materialFunction=&/MF_RingMask}.Result
  )
  Time() {bOverride_Period=true, desc=For testing, period=1}
 ]
}
