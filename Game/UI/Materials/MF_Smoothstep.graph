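# MF_Smoothstep is a material function in the UI materials folder that computes a Hermite smoothstep of Alpha between edges A and B.
# It remaps Alpha from the A-to-B range, clamps it, then applies the cubic t²(3 − 2t) for an eased transition.
# A and B fall back to their preview values for a unit range, while Alpha has no such default.
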MaterialFunction /Game/UI/Materials/MF_Smoothstep
graph {nodes [
  n1 = Constant() {r=3}
  n2 = FunctionInput() {inputName=Alpha, inputType=FunctionInput_Scalar, sortPriority=3, previewValue=<0.5 0 0 1>}
  n3 = FunctionInput() {bUsePreviewValueAsDefault=true, inputName=A, inputType=FunctionInput_Scalar, sortPriority=3}
  n4 = FunctionInput() {bUsePreviewValueAsDefault=true, inputName=B, inputType=FunctionInput_Scalar, sortPriority=3, previewValue=<1 0 0 1>}
  n5 = Clamp(None=Divide(A=Subtract(A=@n2, B=@n3), B=Subtract(A=@n4, B=@n3)))
  n6 = Multiply(A=Constant() {r=2}, B=@n5)
  FunctionOutput(None=Multiply(A=Subtract(A=@n1, B=@n6), B=Multiply(A=@n5, B=@n5)))
 ]}
