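# AAThreshold is a library-exposed UI material function that turns a scalar gradient into an anti-aliased threshold mask.
# It sums the absolute screen-space DDX and DDY of the gradient and scales that by a fudge factor to get the default filter width, which an optional input overrides.
# MF_Smoothstep then blends across the threshold minus and plus that width, keeping the edge soft relative to pixel size.
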MaterialFunction /Game/UI/Materials/AAThreshold
graph {
 properties {bExposeToLibrary=true}
 nodes [
  n1 = FunctionInput() {bUsePreviewValueAsDefault=true, inputName=Threshold, inputType=FunctionInput_Scalar, sortPriority=1, previewValue=<0.65 0 0 1>}
  n2 = FunctionInput() {bUsePreviewValueAsDefault=true, inputName=Gradient, inputType=FunctionInput_Scalar, sortPriority=2}
  n3 = Abs(None=DDX(Value=@n2))
  n4 = Abs(None=DDY(Value=@n2))
  n5 = FunctionInput() {
   bUsePreviewValueAsDefault=true
   inputName=Filter Width Fudge Factor
   inputType=FunctionInput_Scalar
   sortPriority=4
   previewValue=<0.5 0.5 0.5 0.5>
  }
  n6 = FunctionInput(Preview=Multiply(A=Add(A=@n3, B=@n4), B=@n5)) {bUsePreviewValueAsDefault=true, inputName=Filter Width Override, inputType=FunctionInput_Scalar, sortPriority=3}
  FunctionOutput(None=MaterialFunctionCall!(A=Subtract(A=@n1, B=@n6), B=Add(A=@n1, B=@n6), Alpha=@n2) {materialFunction=&/MF_Smoothstep}.Result)
 ]
}
