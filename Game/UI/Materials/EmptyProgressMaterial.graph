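# EmptyProgressMaterial is a translucent UI-domain material, named for an empty progress display.
# It exposes a scalar Percentage parameter, which feeds a Subtract node whose result is clamped by Saturate.
# A standalone Constant node is also present, and the graph shows no links from either branch to the material outputs.
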Material /Game/UI/Materials/EmptyProgressMaterial
graph {
 properties {
  bUsedWithStaticMesh=false
  blendMode=BLEND_Translucent
  materialDomain=MD_UI
  pixelDepthOffsetMode=PDOM_Legacy
  usedShadingModels=Default Lit
 }
 nodes [
  Constant() {r=1}
  Saturate(None=Subtract(A=ScalarParameter() {parameterName=Percentage, sliderMax=1}))
 ]
}
