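# Translucent UI material for a hollow circular progress bar, with an optional outline and a back disc with adjustable alpha.
# Subtracted radial gradients, shaped by SmoothStep, form the outline and progress ring.
# An angular mask from swizzled UVs plus progress cuts the ring, filling or emptying via the Fills switch.
# The Calculate Progress switch selects either the Percentage parameter or a self-advancing value from Time, CountdownStartTime and CountDownTime.
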Material /Game/UI/Materials/HollowCircularProgressBar
graph {
 properties {
  bUsedWithStaticMesh=false
  blendMode=BLEND_Translucent
  materialDomain=MD_UI
  pixelDepthOffsetMode=PDOM_Legacy
  usedShadingModels=Default Lit
 }
 nodes [
  n1 = MaterialFunctionCall(UVs=TextureCoordinate(), CenterPosition=Constant2Vector() {g=0.5, r=0.5}, Radius=Constant() {r=0.5}, Density=Constant() {r=24}) {materialFunction=&RadialGradientExponential}
  n2 = Subtract(
   A=@n1.RadialGradientExponential
   B=MaterialFunctionCall(UVs=TextureCoordinate(), CenterPosition=Constant2Vector() {g=0.5, r=0.5}, Radius=Constant() {r=0.45}, Density=Constant() {r=24}) {materialFunction=&RadialGradientExponential}.RadialGradientExponential
  )
  n3 = ScalarParameter() {defaultValue=0.7, parameterName=OutringSharpness, sliderMax=1}
  n4 = Add(A=@n3) {constB=0.2}
  n5 = ScalarParameter() {parameterName=Back Alpha, sliderMax=1}
  n6 = TextureCoordinate()
  n7 = Constant2Vector() {g=0.5, r=0.5}
  n8 = Constant() {r=0.43}
  n9 = Constant() {r=2}
  n10 = ScalarParameter() {defaultValue=0.01, parameterName=ProgressRingSharpness}
  n11 = Constant() {r=0.02}
  n12 = TextureCoordinate()
  n13 = Subtract(A=@n8, B=ScalarParameter() {defaultValue=0.07, parameterName=ProgressRingSize})
  n14 = Subtract(
   A=MaterialFunctionCall(
    Alpha=MaterialFunctionCall(UVs=@n6, CenterPosition=@n7, Radius=@n8, Density=@n9) {materialFunction=&RadialGradientExponential}.RadialGradientExponential
    Min=@n10
    Max=Add(A=@n10, B=@n11) {constB=0.2}
   ) {materialFunction=&SmoothStep}.Result
   B=MaterialFunctionCall(
    Alpha=MaterialFunctionCall(UVs=@n12, CenterPosition=@n7, Radius=@n13, Density=@n9) {materialFunction=&RadialGradientExponential}.RadialGradientExponential
    Min=@n10
    Max=Add(A=@n10, B=@n11) {constB=0}
   ) {materialFunction=&SmoothStep}.Result
  )
  n15 = Multiply(A=Subtract(A=TextureCoordinate()) {constB=0.5}) {constB=-1}
  n16 = Subtract(A=Time(), B=ScalarParameter() {parameterName=CountdownStartTime})
  n17 = ScalarParameter() {defaultValue=1, parameterName=CountDownTime}
  n18 = StaticSwitchParameter(True=Clamp(None=Divide(A=@n16, B=@n17)), False=ScalarParameter() {parameterName=Percentage}) {
   desc="In Unattended mode you give the progress bar a max time, and a start time, and it automatically cycles up."
   parameterName=Calculate Progress
  }
  n19 = Add(
   A=MaterialFunctionCall(Vector or UVs=MaterialFunctionCall(XY=@n15) {materialFunction=&Swizzle}.YX) {materialFunction=&VectorToRadialValue}.Vector Converted to Angle
   B=@n18
  )
  n20 = StaticSwitchParameter(True=Floor(None=@n19), False=Ceil(None=OneMinus(None=@n19))) {defaultValue=true, desc=Whether the progress bar fills or empties based on the percentage., parameterName=Fills}
  Add(
   A=Multiply(
    A=ScalarParameter() {parameterName=OutlineAlpha, sliderMax=1}
    B=MaterialFunctionCall(Alpha=@n2, Min=@n3, Max=@n4) {materialFunction=&SmoothStep}.Result
   )
   B=Add(A=Multiply(A=@n1.RadialGradientExponential, B=@n5), B=Multiply(A=@n14, B=@n20))
  )
  n21 = OneMinus(None=MaterialFunctionCall!(UVs, CenterPosition, Radius=Constant() {r=0.36}) {materialFunction=/Engine/Functions/Engine_MaterialFunctions01/Gradient/SphereGradient-2D}.Result Diameter)
  n22 = OneMinus(None=MaterialFunctionCall(Radius=Constant() {r=0.4}) {materialFunction=&SphereGradient-2D}.Result Radius)
  Clamp(None=Subtract(A=Power(Base=@n21) {constExponent=3}, B=Power(Base=@n22) {constExponent=3}))
  Time() {bOverride_Period=true, desc=For testing, period=1}
  VectorParameter() {parameterName=Color, defaultValue=<rgba 1 1 1 1>}
 ]
}
