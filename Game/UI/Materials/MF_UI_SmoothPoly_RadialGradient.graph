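# UI material function that draws an aspect-corrected radial gradient, outputting both a two-color RGB blend and a clamped 0-1 mask.
# GetUserInterfaceUV supplies normalized UVs, with V scaled by height/width so the gradient stays circular on non-square widgets.
# Radius is the pixel-size diagonal over width times Radius Scale, feeding RadialGradientExponential with center and density inputs.
# CenterPosition is passed through unscaled, so it operates in the aspect-corrected UV space.
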MaterialFunction /Game/UI/Materials/MF_UI_SmoothPoly_RadialGradient
graph {nodes [
  n1 = FunctionInput() {bUsePreviewValueAsDefault=true, desc=Gradient color for lerp input a, inputName=Col_01, previewValue=<1 0 0 1>}
  n2 = FunctionInput() {bUsePreviewValueAsDefault=true, desc=Gradient color for lerp input b, inputName=Col_02, previewValue=<0 1 0 1>}
  n3 = MaterialFunctionCall() {materialFunction=&GetUserInterfaceUV}
  n4 = Constant() {r=1}
  n5 = MaterialFunctionCall(Float2=@n3.Pixel Size) {materialFunction=&BreakOutFloat2Components}
  n6 = Max(A=@n5.R)
  n7 = Divide(A=@n5.G, B=@n6)
  n8 = FunctionInput() {bUsePreviewValueAsDefault=true, desc="RadialGradientCentre - Default (0, 0)", inputName=CenterPosition, inputType=FunctionInput_Vector2}
  n9 = FunctionInput() {bUsePreviewValueAsDefault=true, description=GradientDensity, inputName=Density, inputType=FunctionInput_Scalar, previewValue=<2 0 0 1>}
  n10 = Distance(A=Constant2Vector(), B=@n3.Pixel Size)
  n11 = FunctionInput() {bUsePreviewValueAsDefault=true, description=Radius, inputName=Radius Scale, inputType=FunctionInput_Scalar, previewValue=<2 0 0 1>}
  n12 = MaterialFunctionCall(
   UVs=Multiply(A=@n3.Normalized UV, B=AppendVector(A=@n4, B=@n7))
   CenterPosition=@n8
   Radius=Multiply(A=Divide(A=@n10, B=@n6), B=@n11)
   Density=@n9
  ) {materialFunction=&RadialGradientExponential}
  FunctionOutput(None=LinearInterpolate(A=ComponentMask(None=@n1) {b=true, g=true, r=true}, B=ComponentMask(None=@n2) {b=true, g=true, r=true}, Alpha=@n12.RadialGradientExponential)) {desc="Outputs Color Gradient (Vector4)", outputName=ColorGradient}
  FunctionOutput(None=Clamp(None=ComponentMask(None=@n12.RadialGradientExponential) {r=true})) {desc="Outputs Gradient (Vector1)", outputName="Gradient (0-1)"}
 ]}
