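# M_UICircle is a translucent UI-domain material that draws a white circle procedurally, without a texture.
# RadialGradientExponential supplies a radial falloff at its default inputs, which SmoothStep turns into a circular mask alongside a constant white color.
# The SmoothStep maximum is derived from the minimum plus a small offset, giving a crisp but slightly softened edge controlled by one threshold constant.
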Material /Game/UI/Materials/M_UICircle
graph {
 properties {
  bUsedWithStaticMesh=false
  blendMode=BLEND_Translucent
  materialDomain=MD_UI
  pixelDepthOffsetMode=PDOM_Legacy
  usedShadingModels=Default Lit
 }
 nodes [
  Constant3Vector() {constant=<rgba 1 1 1 0>}
  n1 = Constant() {r=0.1}
  MaterialFunctionCall(
   Alpha=MaterialFunctionCall() {materialFunction=&RadialGradientExponential}.RadialGradientExponential
   Min=@n1
   Max=Add(A=@n1) {constB=0.01}
  ) {materialFunction=&SmoothStep}
 ]
}
